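# Audio Level Sequence for shot tac4000, carrying the tentacle boss's sound.
# Three master audio tracks, two named Tentacle Rawrrr and one TentacleAudio, overlap in time, with a shorter roar inside two longer sections.
# An event track on the possessed PistolCanary AudioManager has two trigger keys early in that span, with no visible endpoint.
# The TentacleBoss is also bound as a possessable but has no tracks.
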LevelSequence /PistolCanary/Sequence/tac/tac4000/PistolCanary_tac4000_aud
bindings [
 TentacleBoss {id=#1, possessed_class=PistolCanary_TentacleBoss_C, tracks []}
 PistolCanary_AudioManager {
  id=#2
  possessed_class=PistolCanary_AudioManager_C
  tracks [{
    type=MovieSceneEventTrack
    sections [{
      type=MovieSceneEventTriggerSection
      channels [None {
        type=MovieSceneScriptingEventChannel
        keys [[3610, "<Struct 'MovieSceneEvent' (0x852ff5b1720) {}>"], [3710, "<Struct 'MovieSceneEvent' (0x852ff5b17c0) {}>"]]
       }]
     }]
   }]
 }
]
display_rate {denominator=1, numerator=60}
tick_resolution {denominator=1, numerator=24000}
tracks [
 {
  type=MovieSceneAudioTrack
  display_name=Tentacle Rawrrr
  sections [{
    type=MovieSceneAudioSection
    end_frame=4022
    start_frame=3914
    channels [
     Volume {type=MovieSceneScriptingFloatChannel, default=1}
     Pitch {type=MovieSceneScriptingFloatChannel, default=1}
     AttachActor {type=MovieSceneScriptingActorReferenceChannel}
    ]
   }]
 }
 {
  type=MovieSceneAudioTrack
  display_name=TentacleAudio
  sections [{
    type=MovieSceneAudioSection
    end_frame=4226
    start_frame=3599
    channels [
     Volume {type=MovieSceneScriptingFloatChannel, default=1}
     Pitch {type=MovieSceneScriptingFloatChannel, default=1}
     AttachActor {type=MovieSceneScriptingActorReferenceChannel}
    ]
   }]
 }
 {
  type=MovieSceneAudioTrack
  display_name=Tentacle Rawrrr
  sections [{
    type=MovieSceneAudioSection
    end_frame=4221
    start_frame=3594
    channels [
     Volume {type=MovieSceneScriptingFloatChannel, default=1}
     Pitch {type=MovieSceneScriptingFloatChannel, default=1}
     AttachActor {type=MovieSceneScriptingActorReferenceChannel}
    ]
   }]
 }
]
