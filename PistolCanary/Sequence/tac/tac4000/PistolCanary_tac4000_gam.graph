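# Level sequence for shot tac4000 of PistolCanary that stages five ProjectilePortal actors at set points in the world.
# Each portal's spawn track switches it on at a slightly staggered frame, then off near frame 4000.
# Transform tracks hold each portal at a fixed location and yaw with unit scale.
# Each portal also has an event trigger track, whose target functions are not shown, and the sequence has no master tracks.
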LevelSequence /PistolCanary/Sequence/tac/tac4000/PistolCanary_tac4000_gam
bindings [
 PistolCanary_ProjectilePortal5 {
  id=#1
  possessed_class=PistolCanary_ProjectilePortal_C
  tracks [
   {
    type=MovieSceneSpawnTrack
    sections [{
      type=MovieSceneSpawnSection
      channels [None {type=MovieSceneScriptingBoolChannel, default=false, keys [[3600,false],[3697,true],[4000,false]]}]
     }]
   }
   {
    type=MovieScene3DTransformTrack
    sections [{
      type=MovieScene3DTransformSection
      channels [
       Location.X {type=MovieSceneScriptingDoubleChannel, default=14356, const [13824,1]}
       Location.Y {type=MovieSceneScriptingDoubleChannel, default=73440, const [72192,1]}
       Location.Z {type=MovieSceneScriptingDoubleChannel, default=6488, const [6528,1]}
       Rotation.X {type=MovieSceneScriptingDoubleChannel, default=0, const [0,1]}
       Rotation.Y {type=MovieSceneScriptingDoubleChannel, default=0, const [0,1]}
       Rotation.Z {type=MovieSceneScriptingDoubleChannel, default=90, const [106.2364,1]}
       Scale.X {type=MovieSceneScriptingDoubleChannel, default=1, const [1,1]}
       Scale.Y {type=MovieSceneScriptingDoubleChannel, default=1, const [1,1]}
       Scale.Z {type=MovieSceneScriptingDoubleChannel, default=1, const [1,1]}
      ]
     }]
   }
   {
    type=MovieSceneEventTrack
    sections [{
      type=MovieSceneEventTriggerSection
      channels [None {type=MovieSceneScriptingEventChannel, const ["<Struct 'MovieSceneEvent' (0x852ff5b1ea0) {}>", 1]}]
     }]
   }
  ]
 }
 PistolCanary_ProjectilePortal3 {
  id=#2
  possessed_class=PistolCanary_ProjectilePortal_C
  tracks [
   {
    type=MovieSceneSpawnTrack
    sections [{
      type=MovieSceneSpawnSection
      channels [None {type=MovieSceneScriptingBoolChannel, default=false, keys [[3600,false],[3710,true],[4000,false]]}]
     }]
   }
   {
    type=MovieScene3DTransformTrack
    sections [{
      type=MovieScene3DTransformSection
      channels [
       Location.X {type=MovieSceneScriptingDoubleChannel, default=13000, const [12800,1]}
       Location.Y {type=MovieSceneScriptingDoubleChannel, default=72250, const [71168,1]}
       Location.Z {type=MovieSceneScriptingDoubleChannel, default=11232, const [8064,1]}
       Rotation.X {type=MovieSceneScriptingDoubleChannel, default=0, const [0,1]}
       Rotation.Y {type=MovieSceneScriptingDoubleChannel, default=0, const [0,1]}
       Rotation.Z {type=MovieSceneScriptingDoubleChannel, default=120, const [131.9801,1]}
       Scale.X {type=MovieSceneScriptingDoubleChannel, default=1, const [1,1]}
       Scale.Y {type=MovieSceneScriptingDoubleChannel, default=1, const [1,1]}
       Scale.Z {type=MovieSceneScriptingDoubleChannel, default=1, const [1,1]}
      ]
     }]
   }
   {
    type=MovieSceneEventTrack
    sections [{
      type=MovieSceneEventTriggerSection
      channels [None {type=MovieSceneScriptingEventChannel, const ["<Struct 'MovieSceneEvent' (0x852ff5b1400) {}>", 1]}]
     }]
   }
  ]
 }
 PistolCanary_ProjectilePortal {
  id=#3
  possessed_class=PistolCanary_ProjectilePortal_C
  tracks [
   {
    type=MovieSceneSpawnTrack
    sections [{
      type=MovieSceneSpawnSection
      channels [None {type=MovieSceneScriptingBoolChannel, default=false, keys [[3586,false],[3709,true],[3986,false]]}]
     }]
   }
   {
    type=MovieScene3DTransformTrack
    sections [{
      type=MovieScene3DTransformSection
      channels [
       Location.X {type=MovieSceneScriptingDoubleChannel, default=15360, const [15360,1]}
       Location.Y {type=MovieSceneScriptingDoubleChannel, default=70144, const [69632,1]}
       Location.Z {type=MovieSceneScriptingDoubleChannel, default=9216, const [9984,1]}
       Rotation.X {type=MovieSceneScriptingDoubleChannel, default=0, const [0,1]}
       Rotation.Y {type=MovieSceneScriptingDoubleChannel, default=0, const [0,1]}
       Rotation.Z {type=MovieSceneScriptingDoubleChannel, default=36.3016, const [39.5488,1]}
       Scale.X {type=MovieSceneScriptingDoubleChannel, default=1, const [1,1]}
       Scale.Y {type=MovieSceneScriptingDoubleChannel, default=1, const [1,1]}
       Scale.Z {type=MovieSceneScriptingDoubleChannel, default=1, const [1,1]}
      ]
     }]
   }
   {
    type=MovieSceneEventTrack
    sections [{
      type=MovieSceneEventTriggerSection
      channels [None {type=MovieSceneScriptingEventChannel, const ["<Struct 'MovieSceneEvent' (0x852ff5b1a40) {}>", 1]}]
     }]
   }
  ]
 }
 PistolCanary_ProjectilePortal2 {
  id=#4
  possessed_class=PistolCanary_ProjectilePortal_C
  tracks [
   {
    type=MovieSceneSpawnTrack
    sections [{
      type=MovieSceneSpawnSection
      channels [None {type=MovieSceneScriptingBoolChannel, default=false, keys [[3600,false],[3702,true],[4000,false]]}]
     }]
   }
   {
    type=MovieScene3DTransformTrack
    sections [{
      type=MovieScene3DTransformSection
      channels [
       Location.X {type=MovieSceneScriptingDoubleChannel, default=12556, const [12800,1]}
       Location.Y {type=MovieSceneScriptingDoubleChannel, default=71660, const [70656,1]}
       Location.Z {type=MovieSceneScriptingDoubleChannel, default=11232, const [11136,1]}
       Rotation.X {type=MovieSceneScriptingDoubleChannel, default=0, const [0,1]}
       Rotation.Y {type=MovieSceneScriptingDoubleChannel, default=0, const [0,1]}
       Rotation.Z {type=MovieSceneScriptingDoubleChannel, default=160, const [132.7299,1]}
       Scale.X {type=MovieSceneScriptingDoubleChannel, default=1, const [1,1]}
       Scale.Y {type=MovieSceneScriptingDoubleChannel, default=1, const [1,1]}
       Scale.Z {type=MovieSceneScriptingDoubleChannel, default=1, const [1,1]}
      ]
     }]
   }
   {
    type=MovieSceneEventTrack
    sections [{
      type=MovieSceneEventTriggerSection
      channels [None {type=MovieSceneScriptingEventChannel, const ["<Struct 'MovieSceneEvent' (0x852ff5b1360) {}>", 1]}]
     }]
   }
  ]
 }
 PistolCanary_ProjectilePortal4 {
  id=#5
  possessed_class=PistolCanary_ProjectilePortal_C
  tracks [
   {
    type=MovieSceneSpawnTrack
    sections [{
      type=MovieSceneSpawnSection
      channels [None {type=MovieSceneScriptingBoolChannel, default=false, keys [[3600,false],[3706,true],[4000,false]]}]
     }]
   }
   {
    type=MovieScene3DTransformTrack
    sections [{
      type=MovieScene3DTransformSection
      channels [
       Location.X {type=MovieSceneScriptingDoubleChannel, default=17000, const [15360,1]}
       Location.Y {type=MovieSceneScriptingDoubleChannel, default=72500, const [71168,1]}
       Location.Z {type=MovieSceneScriptingDoubleChannel, default=7400, const [7296,1]}
       Rotation.X {type=MovieSceneScriptingDoubleChannel, default=0, const [0,1]}
       Rotation.Y {type=MovieSceneScriptingDoubleChannel, default=0, const [0,1]}
       Rotation.Z {type=MovieSceneScriptingDoubleChannel, default=45, const [54.7069,1]}
       Scale.X {type=MovieSceneScriptingDoubleChannel, default=1, const [1,1]}
       Scale.Y {type=MovieSceneScriptingDoubleChannel, default=1, const [1,1]}
       Scale.Z {type=MovieSceneScriptingDoubleChannel, default=1, const [1,1]}
      ]
     }]
   }
   {
    type=MovieSceneEventTrack
    sections [{
      type=MovieSceneEventTriggerSection
      channels [None {type=MovieSceneScriptingEventChannel, const ["<Struct 'MovieSceneEvent' (0x852ff5b0fa0) {}>", 1]}]
     }]
   }
  ]
 }
]
display_rate {denominator=1, numerator=60}
tick_resolution {denominator=1, numerator=24000}
tracks []
